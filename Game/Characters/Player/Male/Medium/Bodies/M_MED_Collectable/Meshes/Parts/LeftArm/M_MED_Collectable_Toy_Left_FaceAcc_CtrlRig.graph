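# Control Rig for the Collectable toy character's left arm part, driving a dyn_ bone chain from the body skeleton.
# Spherical pose readers and the parachute state stretch the upper arm in X before two-bone IK solves to the hand, using L_Arm_PV as pole.
# Constraints copy hand and fingers, and blend arm and twist bones by variable.
# Aim bones orient elbow, shoulder band and spike parts, adjusted for rifle, emote/pickaxe, parachute and skydive states.
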ControlRigBlueprint /Game/Characters/Player/Male/Medium/Bodies/M_MED_Collectable/Meshes/Parts/LeftArm/M_MED_Collectable_Toy_Left_FaceAcc_CtrlRig
graphs [
 ContainedGraph {nodes [
   n1 = RigVMFunctionEntryNode()
   RigUnit_GetTransform = &RigUnit_GetTransform(Space=LocalSpace, bInitial=false, Item=@n1.Joint)
   n2 = &RigVMFunction_MathQuaternionToEuler(RotationOrder=ZYX, Value=@RigUnit_GetTransform.Transform.Rotation)
   > RigUnit_SetTransform = &RigUnit_SetTransform(
    Space=LocalSpace
    Value=(Rotation=(X=0.0132,Y=0.4099,Z=-0.0222,W=0.9118),Translation=(X=23.6045,Y=-6.37,Z=140.1115),Scale3D=(X=1,Y=1,Z=1))
    bInitial=false
    Item=@n1.Joint
    Weight=@n1.Weight
    Value.Translation.Z=&RigVMFunction_MathFloatAdd(B=@n1.Translate Z, A=@RigUnit_GetTransform.Transform.Translation.Z) {template="Add::Execute(in A,in B,out Result)"}.Result
    Value.Scale3D.Z=&RigVMFunction_MathFloatAdd(B=@n1.Scale Z, A=@RigUnit_GetTransform.Transform.Scale3D.Z) {template="Add::Execute(in A,in B,out Result)"}.Result
    Value.Translation.Y=&RigVMFunction_MathFloatAdd(B=@n1.Translate Y, A=@RigUnit_GetTransform.Transform.Translation.Y) {template="Add::Execute(in A,in B,out Result)"}.Result
    Value.Scale3D.Y=&RigVMFunction_MathFloatAdd(B=@n1.Scale Y, A=@RigUnit_GetTransform.Transform.Scale3D.Y) {template="Add::Execute(in A,in B,out Result)"}.Result
    Value.Translation.X=&RigVMFunction_MathFloatAdd(B=@n1.Translate X, A=@RigUnit_GetTransform.Transform.Translation.X) {template="Add::Execute(in A,in B,out Result)"}.Result
    Value.Scale3D.X=&RigVMFunction_MathFloatAdd(B=@n1.Scale X, A=@RigUnit_GetTransform.Transform.Scale3D.X) {template="Add::Execute(in A,in B,out Result)"}.Result
    Value.Rotation=&RigVMFunction_MathQuaternionFromEuler(
     Euler=(X=0,Y=0,Z=0)
     Euler.X=&RigVMFunction_MathFloatAdd(B=@n1.Rotate X, A=@n2.Result.X) {template="Add::Execute(in A,in B,out Result)"}.Result
     Euler.Y=&RigVMFunction_MathFloatAdd(B=@n1.Rotate Y, A=@n2.Result.Y) {template="Add::Execute(in A,in B,out Result)"}.Result
     Euler.Z=&RigVMFunction_MathFloatAdd(B=@n1.Rotate Z, A=@n2.Result.Z) {template="Add::Execute(in A,in B,out Result)"}.Result
    ).Result
   )
   > RigVMFunctionReturnNode()
  ]}
 RigVMFunctionLibrary {nodes [Set Transforms for Bone = RigVMCollapseNode(
    ExecuteContext=()
    Joint=()
    Rotate X=0
    Rotate Y=0
    Rotate Z=0
    Scale X=0
    Scale Y=0
    Scale Z=0
    Translate X=0
    Translate Y=0
    Translate Z=0
    Weight=1
   )]}
 RigVMModel {nodes [
   n1 = &RigVMFunction_AlphaInterp(
    InRange=(Max=1)
    InterpSpeedDecreasing=15
    InterpSpeedIncreasing=15
    OutRange=(Max=1)
    Scale=1
    bClampResult=true
    Bias=~
    ClampMin=~
    bMapRange=~
    Value=RigVMVariableNode() {variable=Weapon Swap}.Value
   )
   &RigUnit_ParentConstraint(
    AdvancedSettings=(RotationOrderForFilter=XZY)
    Child=(Type=Bone,Name="dyn_arm_l_upperarm_l")
    ExecutePin=()
    Filter=(TranslationFilter=(bX=false,bY=false,bZ=false),RotationFilter=(bX=True,bY=True,bZ=True),ScaleFilter=(bX=false,bY=false,bZ=false))
    Parents=((Item=(Type=Bone,Name="upperarm_l"),Weight=1))
    Weight=@n1.Result
   )
   > &RigUnit_ParentConstraint(
    AdvancedSettings=(RotationOrderForFilter=XZY)
    Child=(Type=Bone,Name="dyn_arm_l_lowerarm_l")
    Filter=(TranslationFilter=(bX=false,bY=false,bZ=false),RotationFilter=(bX=True,bY=True,bZ=True),ScaleFilter=(bX=false,bY=false,bZ=false))
    Parents=((Item=(Type=Bone,Name="lowerarm_l"),Weight=1))
    Weight=1
   )
   RigUnit_BeginExecution = &RigUnit_BeginExecution()
   &RigUnit_ParentConstraint(
    AdvancedSettings=(RotationOrderForFilter=XZY)
    Child=(Type=Bone,Name="dyn_arm_l_clavicle_l")
    Parents=((Item=(Type=Bone,Name="clavicle_l"),Weight=1))
    Weight=0.2
    ExecutePin=@RigUnit_BeginExecution.ExecutePin
   )
   > &RigUnit_ParentConstraint(
    AdvancedSettings=(RotationOrderForFilter=XZY)
    Child=(Type=Bone,Name="dyn_arm_l_hand_l")
    Parents=((Item=(Type=Bone,Name="hand_l"),Weight=1))
    Weight=1
   )
   > n2 = &RigUnit_SphericalPoseReader(
    ActiveRegionScaleFactors=(PositiveWidth=1.25,NegativeWidth=1,PositiveHeight=1,NegativeHeight=1)
    ActiveRegionSize=0.3
    Debug=(bDrawDebug=false,DebugScale=25,DebugSegments=20,DebugThickness=0.25)
    DriverAxis=(X=1,Y=0,Z=0)
    DriverItem=(Type=Bone,Name="upperarm_l")
    FalloffRegionScaleFactors=(PositiveWidth=0.6,NegativeWidth=1.3,PositiveHeight=1,NegativeHeight=1)
    FalloffSize=0.4
    FlipHeightScaling=true
    FlipWidthScaling=true
    OptionalParentItem=(Type=Bone,Name="spine_04")
    RotationOffset=(X=30,Y=15,Z=50)
   )
   > Set Transforms for Bone_1_1_1 = RigVMFunctionReferenceNode(
    Joint=(Type=Bone,Name="dyn_arm_l_upperarm_l")
    Rotate X=0
    Rotate Y=0
    Rotate Z=0
    Scale Y=0
    Scale Z=0
    Translate X=0
    Translate Y=0
    Translate Z=0
    Weight=@n2.OutputParam
    Scale X=RigVMVariableNode() {variable=Rifle}.Value
   )
   n3 = &RigVMFunction_AlphaInterp(
    InRange=(Max=1)
    InterpSpeedDecreasing=10
    InterpSpeedIncreasing=10
    OutRange=(Max=1)
    Scale=1
    Bias=~
    ClampMin=~
    bMapRange=~
    Value=RigVMVariableNode() {variable=Parachute Arm}.Value
   )
   > Set Transforms for Bone_1_1_1_2 = RigVMFunctionReferenceNode(
    Joint=(Type=Bone,Name="dyn_arm_l_upperarm_l")
    Rotate X=0
    Rotate Y=0
    Rotate Z=0
    Scale X=0.1
    Scale Y=0
    Scale Z=0
    Translate X=0
    Translate Y=0
    Translate Z=0
    Weight=@n3.Result
   )
   > n4 = &RigUnit_SphericalPoseReader(
    ActiveRegionScaleFactors=(PositiveWidth=0.4,NegativeWidth=0.3,PositiveHeight=0.3,NegativeHeight=0.8)
    ActiveRegionSize=0.2
    Debug=(bDrawDebug=false,DebugScale=25,DebugSegments=20,DebugThickness=0.25)
    DriverAxis=(X=1,Y=0,Z=0)
    DriverItem=(Type=Bone,Name="upperarm_l")
    FalloffRegionScaleFactors=(PositiveWidth=0.6,NegativeWidth=0.4,PositiveHeight=0.2,NegativeHeight=0.6)
    FalloffSize=0.45
    FlipHeightScaling=true
    FlipWidthScaling=true
    OptionalParentItem=(Type=Bone,Name="spine_04")
    RotationOffset=(X=90,Y=0,Z=10)
   )
   > Set Transforms for Bone_1_1_1_1_1 = RigVMFunctionReferenceNode(
    Joint=(Type=Bone,Name="dyn_arm_l_upperarm_l")
    Rotate X=0
    Rotate Y=0
    Rotate Z=0
    Scale X=0.2
    Scale Y=0
    Scale Z=0
    Translate X=0
    Translate Y=0
    Translate Z=0
    Weight=@n4.OutputParam
   )
   > n5 = &RigUnit_SphericalPoseReader(
    ActiveRegionScaleFactors=(PositiveWidth=0.5,NegativeWidth=0.7,PositiveHeight=0.5,NegativeHeight=0.7)
    ActiveRegionSize=0.15
    Debug=(bDrawDebug=false,DebugScale=25,DebugSegments=20,DebugThickness=0.25)
    DriverAxis=(X=1,Y=0,Z=0)
    DriverItem=(Type=Bone,Name="upperarm_l")
    FalloffRegionScaleFactors=(PositiveWidth=0.6,NegativeWidth=1,PositiveHeight=1,NegativeHeight=1)
    FalloffSize=0.35
    FlipHeightScaling=false
    FlipWidthScaling=false
    OptionalParentItem=(Type=Bone,Name="spine_02")
    RotationOffset=(X=-20,Y=-130,Z=0)
   )
   > Set Transforms for Bone_1_1_1_1_1_1 = RigVMFunctionReferenceNode(
    Joint=(Type=Bone,Name="dyn_arm_l_upperarm_l")
    Rotate X=0
    Rotate Y=0
    Rotate Z=0
    Scale X=0.05
    Scale Y=0
    Scale Z=0
    Translate X=0
    Translate Y=0
    Translate Z=0
    Weight=@n5.OutputParam
   )
   n6 = /Script/ControlRig.RigUnit_ProjectTransformToNewParent(
    Child=(Type=Bone,Name="dyn_arm_l_lowerarm_l")
    NewParent=(Type=Bone,Name="hand_l")
    OldParent=(Type=Bone,Name="lowerarm_l")
    bChildInitial=True
    bOldParentInitial=True
   ) {is=struct RigVMUnitNode}
   > /Script/ControlRig.RigUnit_TwoBoneIKSimplePerItem(
    DebugSettings=(Scale=10,WorldOffset=(Rotation=(X=0,Y=0,Z=0,W=1),Translation=(X=0,Y=0,Z=0),Scale3D=(X=1,Y=1,Z=1)))
    EffectorItem=(Type=Bone,Name="dyn_arm_l_hand_l")
    ItemA=(Type=Bone,Name="dyn_arm_l_upperarm_l")
    ItemB=(Type=Bone,Name="dyn_arm_l_lowerarm_l")
    PoleVectorKind=Location
    PoleVectorSpace=(Type=Bone)
    PrimaryAxis=(X=1,Y=0,Z=0)
    SecondaryAxis=(X=0,Y=1,Z=0)
    SecondaryAxisWeight=0
    StretchMaximumRatio=1.25
    StretchStartRatio=0.75
    Weight=1
    bEnableStretch=false
    bPropagateToChildren=True
    Effector=@n6.Transform
    PoleVector=&RigUnit_GetTransform(Item=(Type=Control,Name="L_Arm_PV")).Transform.Translation
   ) {is=struct RigVMUnitNode}
   n7 = RigVMVariableNode() {variable=Emote And Pickaxe}
   > &RigUnit_ParentConstraint(
    AdvancedSettings=(RotationOrderForFilter=XZY)
    Child=(Type=Bone,Name="dyn_arm_l_upperarm_l")
    Filter=(TranslationFilter=(bX=false,bY=false,bZ=false),RotationFilter=(bX=True,bY=True,bZ=True),ScaleFilter=(bX=false,bY=false,bZ=false))
    Parents=((Item=(Type=Bone,Name="upperarm_l"),Weight=1))
    Weight=@n7.Value
   )
   > &RigUnit_ParentConstraint(
    AdvancedSettings=(RotationOrderForFilter=XZY)
    Child=(Type=Bone,Name="dyn_arm_l_lowerarm_l")
    Filter=(TranslationFilter=(bX=false,bY=false,bZ=false),RotationFilter=(bX=True,bY=True,bZ=True),ScaleFilter=(bX=false,bY=false,bZ=false))
    Parents=((Item=(Type=Bone,Name="lowerarm_l"),Weight=1))
    Weight=@n7.Value
   )
   n8 = RigVMVariableNode() {variable=Arm Twists}
   > &RigUnit_ParentConstraint(
    AdvancedSettings=(RotationOrderForFilter=XZY)
    Child=(Type=Bone,Name="dyn_arm_l_upperarm_twist_01_l")
    Filter=(TranslationFilter=(bX=false,bY=false,bZ=false),RotationFilter=(bX=True,bY=True,bZ=True),ScaleFilter=(bX=True,bY=True,bZ=True))
    Parents=((Item=(Type=Bone,Name="upperarm_twist_01_l"),Weight=1))
    Weight=@n8.Value
   )
   > &RigUnit_ParentConstraint(
    AdvancedSettings=(RotationOrderForFilter=XZY)
    Child=(Type=Bone,Name="dyn_arm_l_upperarm_twist_02_l")
    Filter=(TranslationFilter=(bX=false,bY=false,bZ=false),RotationFilter=(bX=True,bY=True,bZ=True),ScaleFilter=(bX=True,bY=True,bZ=True))
    Parents=((Item=(Type=Bone,Name="upperarm_twist_02_l"),Weight=1))
    Weight=@n8.Value
   )
   > &RigUnit_ParentConstraint(
    AdvancedSettings=(RotationOrderForFilter=XZY)
    Child=(Type=Bone,Name="dyn_arm_l_lowerarm_twist_01_l")
    Filter=(TranslationFilter=(bX=false,bY=false,bZ=false),RotationFilter=(bX=True,bY=True,bZ=True),ScaleFilter=(bX=True,bY=True,bZ=True))
    Parents=((Item=(Type=Bone,Name="lowerarm_twist_01_l"),Weight=1))
    Weight=@n8.Value
   )
   > &RigUnit_ParentConstraint(
    AdvancedSettings=(RotationOrderForFilter=XZY)
    Child=(Type=Bone,Name="dyn_arm_l_lowerarm_twist_02_l")
    Filter=(TranslationFilter=(bX=false,bY=false,bZ=false),RotationFilter=(bX=True,bY=True,bZ=True),ScaleFilter=(bX=True,bY=True,bZ=True))
    Parents=((Item=(Type=Bone,Name="lowerarm_twist_02_l"),Weight=1))
    Weight=@n8.Value
   )
   > &RigUnit_ParentConstraint(
    AdvancedSettings=(RotationOrderForFilter=XZY)
    Child=(Type=Bone,Name="dyn_arm_l_hand_l")
    Parents=((Item=(Type=Bone,Name="hand_l"),Weight=1))
    Weight=1
   )
   > &RigUnit_ParentConstraint(
    AdvancedSettings=(RotationOrderForFilter=XZY)
    Child=(Type=Bone,Name="dyn_thumb_01_l")
    Parents=((Item=(Type=Bone,Name="thumb_01_l"),Weight=1))
    Weight=1
   )
   > &RigUnit_ParentConstraint(
    AdvancedSettings=(RotationOrderForFilter=XZY)
    Child=(Type=Bone,Name="dyn_thumb_02_l")
    Parents=((Item=(Type=Bone,Name="thumb_02_l"),Weight=1))
    Weight=1
   )
   > &RigUnit_ParentConstraint(
    AdvancedSettings=(RotationOrderForFilter=XZY)
    Child=(Type=Bone,Name="dyn_thumb_03_l")
    Parents=((Item=(Type=Bone,Name="thumb_03_l"),Weight=1))
    Weight=1
   )
   > &RigUnit_ParentConstraint(
    AdvancedSettings=(RotationOrderForFilter=XZY)
    Child=(Type=Bone,Name="dyn_index_metacarpal_l")
    Parents=((Item=(Type=Bone,Name="index_metacarpal_l"),Weight=1))
    Weight=1
   )
   > &RigUnit_ParentConstraint(
    AdvancedSettings=(RotationOrderForFilter=XZY)
    Child=(Type=Bone,Name="dyn_index_01_l")
    Parents=((Item=(Type=Bone,Name="index_01_l"),Weight=1))
    Weight=1
   )
   > &RigUnit_ParentConstraint(
    AdvancedSettings=(RotationOrderForFilter=XZY)
    Child=(Type=Bone,Name="dyn_index_02_l")
    Parents=((Item=(Type=Bone,Name="index_02_l"),Weight=1))
    Weight=1
   )
   > &RigUnit_ParentConstraint(
    AdvancedSettings=(RotationOrderForFilter=XZY)
    Child=(Type=Bone,Name="dyn_index_03_l")
    Parents=((Item=(Type=Bone,Name="index_03_l"),Weight=1))
    Weight=1
   )
   > &RigUnit_ParentConstraint(
    AdvancedSettings=(RotationOrderForFilter=XZY)
    Child=(Type=Bone,Name="dyn_middle_metacarpal_l")
    Parents=((Item=(Type=Bone,Name="middle_metacarpal_l"),Weight=1))
    Weight=1
   )
   > &RigUnit_ParentConstraint(
    AdvancedSettings=(RotationOrderForFilter=XZY)
    Child=(Type=Bone,Name="dyn_middle_01_l")
    Parents=((Item=(Type=Bone,Name="middle_01_l"),Weight=1))
    Weight=1
   )
   > &RigUnit_ParentConstraint(
    AdvancedSettings=(RotationOrderForFilter=XZY)
    Child=(Type=Bone,Name="dyn_middle_02_l")
    Parents=((Item=(Type=Bone,Name="middle_02_l"),Weight=1))
    Weight=1
   )
   > &RigUnit_ParentConstraint(
    AdvancedSettings=(RotationOrderForFilter=XZY)
    Child=(Type=Bone,Name="dyn_middle_03_l")
    Parents=((Item=(Type=Bone,Name="middle_03_l"),Weight=1))
    Weight=1
   )
   > &RigUnit_ParentConstraint(
    AdvancedSettings=(RotationOrderForFilter=XZY)
    Child=(Type=Bone,Name="dyn_ring_metacarpal_l")
    Parents=((Item=(Type=Bone,Name="ring_metacarpal_l"),Weight=1))
    Weight=1
   )
   > &RigUnit_ParentConstraint(
    AdvancedSettings=(RotationOrderForFilter=XZY)
    Child=(Type=Bone,Name="dyn_ring_01_l")
    Parents=((Item=(Type=Bone,Name="ring_01_l"),Weight=1))
    Weight=1
   )
   > &RigUnit_ParentConstraint(
    AdvancedSettings=(RotationOrderForFilter=XZY)
    Child=(Type=Bone,Name="dyn_ring_02_l")
    Parents=((Item=(Type=Bone,Name="ring_02_l"),Weight=1))
    Weight=1
   )
   > &RigUnit_ParentConstraint(
    AdvancedSettings=(RotationOrderForFilter=XZY)
    Child=(Type=Bone,Name="dyn_ring_03_l")
    Parents=((Item=(Type=Bone,Name="ring_03_l"),Weight=1))
    Weight=1
   )
   > &RigUnit_ParentConstraint(
    AdvancedSettings=(RotationOrderForFilter=XZY)
    Child=(Type=Bone,Name="dyn_pinky_metacarpal_l")
    Parents=((Item=(Type=Bone,Name="pinky_metacarpal_l"),Weight=1))
    Weight=1
   )
   > &RigUnit_ParentConstraint(
    AdvancedSettings=(RotationOrderForFilter=XZY)
    Child=(Type=Bone,Name="dyn_pinky_01_l")
    Parents=((Item=(Type=Bone,Name="pinky_01_l"),Weight=1))
    Weight=1
   )
   > &RigUnit_ParentConstraint(
    AdvancedSettings=(RotationOrderForFilter=XZY)
    Child=(Type=Bone,Name="dyn_pinky_02_l")
    Parents=((Item=(Type=Bone,Name="pinky_02_l"),Weight=1))
    Weight=1
   )
   > &RigUnit_ParentConstraint(
    AdvancedSettings=(RotationOrderForFilter=XZY)
    Child=(Type=Bone,Name="dyn_pinky_03_l")
    Parents=((Item=(Type=Bone,Name="pinky_03_l"),Weight=1))
    Weight=1
   )
   > /Script/ControlRig.RigUnit_AimItem(
    DebugSettings=(Scale=10,WorldOffset=(Rotation=(X=0,Y=0,Z=0,W=1),Translation=(X=0,Y=0,Z=0),Scale3D=(X=1,Y=1,Z=1)))
    Item=(Type=Bone,Name="dyn_lowerarm_l_aim_MSTR")
    Primary=(Weight=1,Axis=(X=1,Y=0,Z=0),Target=(X=0,Y=0,Z=0),Kind=Location,Space=(Type=Bone,Name="dyn_arm_l_TGT"))
    Secondary=(Weight=1,Axis=(X=0,Y=1,Z=0),Target=(X=0,Y=1,Z=0),Kind=Location,Space=(Type=Bone,Name="upperarm_l"))
    Weight=1
   ) {is=struct RigVMUnitNode}
   > &RigUnit_ModifyTransforms(
    ItemToModify=((Item=(Type=Bone,Name="dyn_lowerarm_l_aim"),Transform=(Rotation=(X=0,Y=0,Z=0,W=1),Translation=(X=0,Y=0,Z=0),Scale3D=(X=1,Y=1,Z=1))))
    ItemToModify.0.Transform.Rotation=&RigVMFunction_MathQuaternionFromEuler(Euler=(X=-20,Y=0,Z=0)).Result
    Weight=RigVMVariableNode() {variable=Rifle Arm Up}.Value
   )
   n9 = &RigVMFunction_AlphaInterp(
    InRange=(Max=1)
    InterpSpeedDecreasing=10
    InterpSpeedIncreasing=10
    OutRange=(Max=1)
    Scale=1
    Bias=~
    ClampMin=~
    bMapRange=~
    Value=RigVMVariableNode() {variable=skydive float}.Value
   )
   > &RigUnit_AimItem(
    DebugSettings=(Scale=10,WorldOffset=(Rotation=(X=0,Y=0,Z=0,W=1),Translation=(X=0,Y=0,Z=0),Scale3D=(X=1,Y=1,Z=1)))
    Item=(Type=Bone,Name="dyn_upperarm_l_AIM_MSTR")
    Primary=(Weight=1,Axis=(X=1,Y=0,Z=0),Target=(X=0,Y=0,Z=0),Kind=Location,Space=(Type=Bone,Name="dyn_arm_l_TGT"))
    Secondary=(Weight=1,Axis=(X=0,Y=0,Z=-1),Target=(X=0,Y=0,Z=0),Kind=Location,Space=(Type=Bone,Name="pelvis"))
    Weight=1
    Secondary.Target.X=&RigVMFunction_MathFloatRemap(SourceMaximum=1, TargetMaximum=0, TargetMinimum=30, bClamp=true, Value=@n9.Result).Result
   )
   n10 = &RigVMFunction_AlphaInterp(
    InRange=(Max=1)
    InterpSpeedDecreasing=10
    InterpSpeedIncreasing=10
    OutRange=(Max=1)
    Scale=1
    Bias=~
    ClampMin=~
    bMapRange=~
    Value=RigVMVariableNode() {variable=skydive float}.Value
   )
   > &RigUnit_ModifyTransforms(
    ItemToModify=((Item=(Type=Bone,Name="dyn_upperarm_l_AIM"),Transform=(Rotation=(X=0,Y=0,Z=0,W=1),Translation=(X=0,Y=0,Z=0),Scale3D=(X=1,Y=1,Z=1))))
    Weight=@n10.Result
    ItemToModify.0.Transform.Rotation=&RigVMFunction_MathQuaternionFromEuler(Euler=(X=90,Y=0,Z=0)).Result
   )
   n11 = RigVMVariableNode() {variable=Emote And Pickaxe}
   > &RigUnit_ParentConstraint(
    AdvancedSettings=(RotationOrderForFilter=XZY)
    Child=(Type=Bone,Name="dyn_upperarm_l_AIM_MSTR")
    Filter=(TranslationFilter=(bX=True,bY=True,bZ=True),RotationFilter=(bX=True,bY=True,bZ=True),ScaleFilter=(bX=false,bY=false,bZ=false))
    Parents=((Item=(Type=Bone,Name="dyn_arm_l_upperarm_l"),Weight=1))
    Weight=@n11.Value
   )
   AlphaInterp = &RigVMFunction_AlphaInterp(
    InRange=(Max=1)
    InterpSpeedDecreasing=7
    InterpSpeedIncreasing=20
    OutRange=(Max=1)
    Scale=1
    Bias=~
    ClampMin=~
    bMapRange=~
    Value=@n11.Value
   )
   > &RigUnit_ParentConstraint(
    AdvancedSettings=(RotationOrderForFilter=XZY)
    Child=(Type=Bone,Name="dyn_lowerarm_l_aim_MSTR")
    Parents=((Item=(Type=Bone,Name="lowerarm_l"),Weight=1))
    Weight=@AlphaInterp.Result
   )
   n12 = &RigVMFunction_AlphaInterp(
    InRange=(Max=1)
    InterpSpeedDecreasing=10
    InterpSpeedIncreasing=10
    OutRange=(Max=1)
    Scale=1
    Bias=~
    ClampMin=~
    bMapRange=~
    Value=RigVMVariableNode() {variable=Parachute Arm}.Value
   )
   > &RigUnit_ParentConstraint(
    AdvancedSettings=(RotationOrderForFilter=XZY)
    Child=(Type=Bone,Name="dyn_upperarm_l_AIM_MSTR")
    Filter=(TranslationFilter=(bX=false,bY=false,bZ=false),RotationFilter=(bX=True,bY=True,bZ=True),ScaleFilter=(bX=false,bY=false,bZ=false))
    Parents=((Item=(Type=Bone,Name="upperarm_l"),Weight=1))
    Weight=@n12.Result
   )
   > &RigUnit_ParentConstraint(
    AdvancedSettings=(RotationOrderForFilter=XZY)
    Child=(Type=Bone,Name="dyn_upper_arm_l_elbow_MSTR")
    Parents=((Item=(Type=Bone,Name="dyn_lowerarm_l_aim"),Weight=1))
    Weight=1
   )
   > &RigUnit_ParentConstraint(
    AdvancedSettings=(RotationOrderForFilter=XZY)
    Child=(Type=Bone,Name="dyn_upper_arm_l_elbow")
    Filter=(TranslationFilter=(bX=false,bY=false,bZ=false),RotationFilter=(bX=false,bY=false,bZ=True),ScaleFilter=(bX=True,bY=True,bZ=True))
    Parents=((Item=(Type=Bone,Name="dyn_arm_l_upperarm_l"),Weight=1))
    Weight=1
   )
   > &RigUnit_ParentConstraint(
    AdvancedSettings=(RotationOrderForFilter=XZY)
    Child=(Type=Bone,Name="dyn_l_shoulder_band")
    Parents=((Item=(Type=Bone,Name="dyn_upperarm_l_AIM"),Weight=1))
    Weight=1
   )
   &RigUnit_RotationConstraintLocalSpaceOffset(
    AdvancedSettings=(RotationOrderForFilter=XZY)
    Child=(Type=Bone,Name="dyn_arm_l_upperarm_twist_01_l")
    ExecutePin=()
    Filter=(bX=True,bY=True,bZ=True)
    Parents=((Item=(Type=Bone,Name="upperarm_twist_01_l"),Weight=1))
    Weight=0
   )
   > &RigUnit_RotationConstraintLocalSpaceOffset(
    AdvancedSettings=(RotationOrderForFilter=XZY)
    Child=(Type=Bone,Name="dyn_arm_l_upperarm_twist_02_l")
    Filter=(bX=True,bY=True,bZ=True)
    Parents=((Item=(Type=Bone,Name="upperarm_twist_02_l"),Weight=1))
    Weight=0
   )
   > &RigUnit_RotationConstraintLocalSpaceOffset(
    AdvancedSettings=(RotationOrderForFilter=XZY)
    Child=(Type=Bone,Name="dyn_arm_l_lowerarm_twist_01_l")
    Filter=(bX=True,bY=True,bZ=True)
    Parents=((Item=(Type=Bone,Name="lowerarm_twist_01_l"),Weight=1))
    Weight=0
   )
   > &RigUnit_RotationConstraintLocalSpaceOffset(
    AdvancedSettings=(RotationOrderForFilter=XZY)
    Child=(Type=Bone,Name="dyn_arm_l_lowerarm_twist_02_l")
    Filter=(bX=True,bY=True,bZ=True)
    Parents=((Item=(Type=Bone,Name="lowerarm_twist_02_l"),Weight=1))
    Weight=0
   )
   &RigUnit_AimItem(
    DebugSettings=(Scale=10,WorldOffset=(Rotation=(X=0,Y=0,Z=0,W=1),Translation=(X=0,Y=0,Z=0),Scale3D=(X=1,Y=1,Z=1)))
    ExecutePin=()
    Item=(Type=Bone,Name="dyn_upperarm_l_AIM")
    Primary=(Weight=1,Axis=(X=1,Y=0,Z=0),Target=(X=0,Y=0,Z=0),Kind=Location,Space=(Type=Bone,Name="dyn_arm_l_lowerarm_l"))
    Secondary=(Weight=1,Axis=(X=1,Y=0,Z=0),Target=(X=0,Y=0,Z=0),Kind=Location,Space=(Type=Bone,Name="dyn_upper_arm_l_elbow_MSTR"))
    Weight=1
   )
   &RigUnit_AimItem(
    DebugSettings=(Scale=10,WorldOffset=(Rotation=(X=0,Y=0,Z=0,W=1),Translation=(X=0,Y=0,Z=0),Scale3D=(X=1,Y=1,Z=1)))
    ExecutePin=()
    Item=(Type=Bone,Name="dyn_upperarm_l_AIM")
    Primary=(Weight=1,Axis=(X=1,Y=0,Z=0),Target=(X=0,Y=0,Z=0),Kind=Location,Space=(Type=Bone,Name="dyn_arm_l_lowerarm_l"))
    Secondary=(Weight=1,Axis=(X=0,Y=1,Z=0),Target=(X=0,Y=0,Z=0),Kind=Location,Space=(Type=Bone,Name="dyn_lowerarm_l_aim"))
    Weight=1
   )
   &RigUnit_AimItem(
    DebugSettings=(Scale=10,WorldOffset=(Rotation=(X=0,Y=0,Z=0,W=1),Translation=(X=0,Y=0,Z=0),Scale3D=(X=1,Y=1,Z=1)))
    ExecutePin=()
    Item=(Type=Bone,Name="dyn_upperarm_l_AIM")
    Primary=(Weight=1,Axis=(X=1,Y=0,Z=0),Target=(X=0,Y=0,Z=0),Kind=Location,Space=(Type=Bone,Name="dyn_arm_l_TGT"))
    Secondary=(Weight=1,Axis=(X=0,Y=0,Z=1),Target=(X=0,Y=0,Z=1),Kind=Location,Space=(Type=Bone,Name="lowerarm_l"))
    Weight=1
   )
   &RigVMFunction_AlphaInterp(InRange=(Max=1), InterpSpeedDecreasing=10, InterpSpeedIncreasing=25, OutRange=(Max=1), Scale=1, Bias=~, ClampMin=~, bMapRange=~)
   /Script/ControlRig.RigUnit_DebugTransformMutableItemSpace(
    Color=(R=1,G=0,B=0,A=1)
    DebugDrawSettings=(DepthPriority=SDPG_Foreground,Lifetime=-1)
    ExecutePin=()
    Mode=Point
    Scale=5
    Space=(Type=Bone)
    Thickness=1
    WorldOffset=(Rotation=(X=0,Y=0,Z=0,W=1),Translation=(X=0,Y=0,Z=0),Scale3D=(X=1,Y=1,Z=1))
    bEnabled=True
    Transform=&RigUnit_GetTransform(Item=(Type=Bone,Name="dyn_upperarm_l_AIM")).Transform
   ) {is=struct RigVMUnitNode}
   &RigVMFunction_MathFloatRemap(SourceMaximum=1, TargetMaximum=0.1, TargetMinimum=0, bClamp=true, Value=@n3.Result)
   &RigUnit_ParentConstraint(
    AdvancedSettings=(RotationOrderForFilter=XZY)
    Child=(Type=Bone,Name="dyn_upper_arm_l_elbow")
    ExecutePin=()
    Filter=(TranslationFilter=(bX=false,bY=false,bZ=false),RotationFilter=(bX=True,bY=True,bZ=True),ScaleFilter=(bX=True,bY=True,bZ=True))
    Parents=((Item=(Type=Bone,Name="dyn_upperarm_l_AIM"),Weight=1))
    Weight=1
   )
   &RigUnit_ParentConstraint(
    AdvancedSettings=(RotationOrderForFilter=XZY)
    Child=(Type=Bone,Name="dyn_upper_arm_l_elbow")
    ExecutePin=()
    Filter=(TranslationFilter=(bX=false,bY=false,bZ=false),RotationFilter=(bX=True,bY=True,bZ=True),ScaleFilter=(bX=false,bY=false,bZ=false))
    Parents=((Item=(Type=Bone,Name="dyn_arm_l_upperarm_l"),Weight=1))
    Weight=1
   )
   &RigUnit_ParentConstraint(
    AdvancedSettings=(RotationOrderForFilter=XZY)
    Child=(Type=Bone,Name="dyn_upperarm_l_AIM")
    ExecutePin=()
    Parents=((Item=(Type=Bone,Name="dyn_arm_l_upperarm_l"),Weight=1))
    Weight=1
   )
   &RigUnit_ProjectTransformToNewParent(
    Child=(Type=Bone,Name="dyn_arm_l_lowerarm_l")
    NewParent=(Type=Bone,Name="lowerarm_l")
    OldParent=(Type=Bone,Name="lowerarm_l")
    bChildInitial=True
    bOldParentInitial=True
   )
   n13 = &RigUnit_SphericalPoseReader(
    ActiveRegionScaleFactors=(PositiveWidth=1.5,NegativeWidth=1,PositiveHeight=1.5,NegativeHeight=1)
    ActiveRegionSize=0.2
    Debug=(bDrawDebug=True,DebugScale=25,DebugSegments=20,DebugThickness=0.25)
    DriverAxis=(X=1,Y=0,Z=0)
    DriverItem=(Type=Bone,Name="upperarm_l")
    ExecutePin=()
    FalloffRegionScaleFactors=(PositiveWidth=0.6,NegativeWidth=1.3,PositiveHeight=2,NegativeHeight=1)
    FalloffSize=0.4
    FlipHeightScaling=true
    FlipWidthScaling=true
    OptionalParentItem=(Type=Bone,Name="spine_04")
    RotationOffset=(X=80,Y=0,Z=-10)
   )
   Set Transforms for Bone_1_1_1_1 = RigVMFunctionReferenceNode(
    ExecuteContext=()
    Joint=(Type=Bone,Name="dyn_arm_l_upperarm_l")
    Rotate X=0
    Rotate Y=0
    Rotate Z=0
    Scale Y=0
    Scale Z=0
    Translate X=0
    Translate Y=0
    Translate Z=0
    Weight=@n13.OutputParam
    Scale X=RigVMVariableNode() {variable=Dual Wield}.Value
   )
   RigVMVariableNode() {variable=skydive float}
   RigVMVariableNode() {variable=skydive float}
  ]}
]
hierarchy [
 root {loc=<0 0 0>, rot=<0 0 0 1>}
 pelvis {parent=root, loc=<0 -0.7071 93.1191>, rot=<0.0127 0.707 -0.0127 -0.707>}
 spine_01 {parent=pelvis, loc=<6.7494 -0.6388 0>, rot=<0 0 0.034 -0.9994>}
 spine_02 {parent=spine_01, loc=<7.2518 -0.035 0>, rot=<0 0 0.0238 -0.9997>}
 spine_03 {parent=spine_02, loc=<6.0104 0 0>, rot=<0 0 0.0149 -0.9999>}
 spine_04 {parent=spine_03, loc=<8.1024 0 0>, rot=<0 0 -0.0132 -0.9999>}
 spine_05 {parent=spine_04, loc=<10.0146 0 0>, rot=<0 0 -0.0015 -1>}
 clavicle_l {parent=spine_05, loc=<4.3575 -0.1466 -2.3223>, rot=<0.0465 -0.7289 -0.009 -0.6829>}
 upperarm_l {parent=clavicle_l, loc=<15.5111 0 0>, rot=<0.0226 -0.3234 -0.0088 -0.9459>}
 lowerarm_l {parent=upperarm_l, loc=<26.6514 0 0>, rot=<0.0391 -0.0099 -0.2453 -0.9686>}
 hand_l {parent=lowerarm_l, loc=<22.3491 0 0>, rot=<-0.7081 -0.0507 0.0975 -0.6975>}
 index_metacarpal_l {parent=hand_l, loc=<1.6535 0.0183 -2.96>, rot=<0.0158 -0.0124 -0.0173 -0.9996>}
 index_01_l {parent=index_metacarpal_l, loc=<8.3008 0.0357 -0.0524>, rot=<-0.0044 0.0485 0.0924 -0.9945>}
 index_02_l {parent=index_01_l, loc=<3.3587 -0.3004 0.0452>, rot=<-0.002 0.0027 0.3225 -0.9466>}
 index_03_l {parent=index_02_l, loc=<2.5578 -0.123 -0.0252>, rot=<-0.002 0.0067 0.2558 -0.9667>}
 middle_metacarpal_l {parent=hand_l, loc=<2.1856 -0.1725 -0.5984>, rot=<0.006 0.0168 -0.0329 -0.9993>}
 middle_01_l {parent=middle_metacarpal_l, loc=<7.3275 -0.2447 0.0129>, rot=<-0.0208 0.0042 0.0858 -0.9961>}
 middle_02_l {parent=middle_01_l, loc=<4.0123 0.1401 0.1175>, rot=<-0.0218 0.0258 0.3512 -0.9357>}
 middle_03_l {parent=middle_02_l, loc=<2.6077 0.0534 0.1373>, rot=<0.0082 0.0188 0.1708 -0.9851>}
 pinky_metacarpal_l {parent=hand_l, loc=<1.7616 -0.9753 3.3814>, rot=<-0.0397 0.0814 -0.069 -0.9935>}
 pinky_01_l {parent=pinky_metacarpal_l, loc=<7.1709 -0.4831 -0.1094>, rot=<-0.0654 -0.0035 0.2242 -0.9723>}
 pinky_02_l {parent=pinky_01_l, loc=<2.8274 0.0893 -0.1043>, rot=<0.0037 0.0011 0.2854 -0.9584>}
 pinky_03_l {parent=pinky_02_l, loc=<2.2046 0.0307 -0.0078>, rot=<-0.0002 0.0098 0.1445 -0.9895>}
 ring_metacarpal_l {parent=hand_l, loc=<2.2074 -0.3625 1.3638>, rot=<-0.0284 0.0722 -0.0458 -0.9959>}
 ring_01_l {parent=ring_metacarpal_l, loc=<7.0295 -0.6197 -0.1446>, rot=<-0.0489 -0.0252 0.1892 -0.9804>}
 ring_02_l {parent=ring_01_l, loc=<3.6174 0.4647 -0.1027>, rot=<0 0 0.2806 -0.9598>}
 ring_03_l {parent=ring_02_l, loc=<2.4152 0.114 0.0199>, rot=<0 0 0.1189 -0.9929>}
 thumb_01_l {parent=hand_l, loc=<1.3118 -0.3191 -2.459>, rot=<0.6546 -0.4042 0.0664 -0.6354>}
 thumb_02_l {parent=thumb_01_l, loc=<5.9336 -0.0448 0.0401>, rot=<-0.0037 0.0213 0.298 -0.9543>}
 thumb_03_l {parent=thumb_02_l, loc=<2.2223 0.3682 -0.0907>, rot=<0.0118 -0.013 0.0066 -0.9998>}
 weapon_l {parent=hand_l, loc=<-0.1974 -3.4111 0.3811>, rot=<0 0 0 1>}
 lowerarm_twist_01_l {parent=lowerarm_l, loc=<17.8549 0 0>, rot=<0 0 0 1>}
 lowerarm_twist_02_l {parent=lowerarm_l, loc=<9.6199 0 0>, rot=<0 0 0 1>}
 L_Arm_PV {controlType=EULER_TRANSFORM, parent=lowerarm_l, type=Control, loc=<0 -20 0>, rot=<0 0 0 1>}
 upperarm_twist_01_l {parent=upperarm_l, loc=<0.0041 0 0>, rot=<0 0 0 1>}
 upperarm_twist_02_l {parent=upperarm_l, loc=<13.7941 0 0>, rot=<0 0 0 1>}
 dyn_spine_05 {parent=spine_05, loc=<0 0 0>, rot=<-0.0039 -0.7071 -0.7071 -0.0039>}
 dyn_arm_l_clavicle_l {parent=dyn_spine_05, loc=<-4.3588 -2.3223 -0.0988>, rot=<-0.5062 0.4472 0.5186 -0.5242>}
 dyn_arm_l_upperarm_l {parent=dyn_arm_l_clavicle_l, loc=<15.5111 0 0>, rot=<0.0226 -0.3234 -0.0088 -0.9459>}
 dyn_arm_l_lowerarm_l {parent=dyn_arm_l_upperarm_l, loc=<26.6514 0 0>, rot=<0.0391 -0.0099 -0.2453 -0.9686>}
 dyn_arm_l_hand_l {parent=dyn_arm_l_lowerarm_l, loc=<22.3491 0 0>, rot=<-0.7081 -0.0507 0.0975 -0.6975>}
 dyn_thumb_01_l {parent=dyn_arm_l_hand_l, loc=<1.3118 -0.3191 -2.459>, rot=<0.6546 -0.4042 0.0664 -0.6354>}
 dyn_thumb_02_l {parent=dyn_thumb_01_l, loc=<5.9336 -0.0448 0.0401>, rot=<-0.0037 0.0213 0.298 -0.9543>}
 dyn_thumb_03_l {parent=dyn_thumb_02_l, loc=<2.2223 0.3682 -0.0907>, rot=<0.0118 -0.013 0.0066 -0.9998>}
 dyn_ring_metacarpal_l {parent=dyn_arm_l_hand_l, loc=<2.2074 -0.3625 1.3638>, rot=<-0.0284 0.0722 -0.0458 -0.9959>}
 dyn_ring_01_l {parent=dyn_ring_metacarpal_l, loc=<7.0295 -0.6197 -0.1446>, rot=<-0.0489 -0.0252 0.1892 -0.9804>}
 dyn_ring_02_l {parent=dyn_ring_01_l, loc=<3.6174 0.4647 -0.1027>, rot=<0 0 0.2806 -0.9598>}
 dyn_ring_03_l {parent=dyn_ring_02_l, loc=<2.4152 0.114 0.0199>, rot=<0 0 0.1189 -0.9929>}
 dyn_pinky_metacarpal_l {parent=dyn_arm_l_hand_l, loc=<1.7616 -0.9753 3.3814>, rot=<-0.0397 0.0814 -0.069 -0.9935>}
 dyn_pinky_01_l {parent=dyn_pinky_metacarpal_l, loc=<7.1709 -0.4831 -0.1094>, rot=<-0.0654 -0.0035 0.2242 -0.9723>}
 dyn_pinky_02_l {parent=dyn_pinky_01_l, loc=<2.8274 0.0893 -0.1043>, rot=<0.0037 0.0011 0.2854 -0.9584>}
 dyn_pinky_03_l {parent=dyn_pinky_02_l, loc=<2.2046 0.0307 -0.0078>, rot=<-0.0002 0.0098 0.1445 -0.9895>}
 dyn_middle_metacarpal_l {parent=dyn_arm_l_hand_l, loc=<2.1856 -0.1725 -0.5984>, rot=<0.006 0.0168 -0.0329 -0.9993>}
 dyn_middle_01_l {parent=dyn_middle_metacarpal_l, loc=<7.3275 -0.2447 0.0129>, rot=<-0.0208 0.0042 0.0858 -0.9961>}
 dyn_middle_02_l {parent=dyn_middle_01_l, loc=<4.0123 0.1401 0.1175>, rot=<-0.0218 0.0258 0.3512 -0.9357>}
 dyn_middle_03_l {parent=dyn_middle_02_l, loc=<2.6077 0.0534 0.1373>, rot=<0.0082 0.0188 0.1708 -0.9851>}
 dyn_index_metacarpal_l {parent=dyn_arm_l_hand_l, loc=<1.6535 0.0183 -2.96>, rot=<0.0158 -0.0124 -0.0173 -0.9996>}
 dyn_index_01_l {parent=dyn_index_metacarpal_l, loc=<8.3008 0.0357 -0.0524>, rot=<-0.0044 0.0485 0.0924 -0.9945>}
 dyn_index_02_l {parent=dyn_index_01_l, loc=<3.3587 -0.3004 0.0452>, rot=<-0.002 0.0027 0.3225 -0.9466>}
 dyn_index_03_l {parent=dyn_index_02_l, loc=<2.5578 -0.123 -0.0252>, rot=<-0.002 0.0067 0.2558 -0.9667>}
 dyn_lowerarm_l_aim_MSTR {parent=dyn_arm_l_hand_l, loc=<0 0 0>, rot=<0.1001 -0.7342 0.6699 -0.0454>}
 dyn_lowerarm_l_aim {parent=dyn_lowerarm_l_aim_MSTR, loc=<0 0 0>, rot=<0 0 0 1>}
 dyn_arm_l_lowerarm_twist_01_l {parent=dyn_arm_l_lowerarm_l, loc=<17.8549 0 0>, rot=<0 0 0 1>}
 dyn_arm_l_lowerarm_twist_02_l {parent=dyn_arm_l_lowerarm_l, loc=<9.6199 0 0>, rot=<0 0 0 1>}
 dyn_arm_l_TGT {parent=dyn_arm_l_lowerarm_l, loc=<0 0 0>, rot=<-0.0391 0.0099 0.2453 -0.9686>}
 dyn_upper_arm_l_elbow_MSTR {parent=dyn_arm_l_TGT, loc=<0 0 0>, rot=<0 -1 -0.0075 0>}
 dyn_upper_arm_l_elbow {parent=dyn_upper_arm_l_elbow_MSTR, loc=<0 0 0>, rot=<0 -1 -0.0075 0>}
 dyn_arm_l_upperarm_twist_01_l {parent=dyn_arm_l_upperarm_l, loc=<0.0041 0 0>, rot=<0 0 0 1>}
 dyn_l_shoulder_band {parent=dyn_arm_l_upperarm_twist_01_l, loc=<2.1628 0.0641 1.3622>, rot=<-0.0144 0.284 -0.0076 -0.9587>}
 dyn_arm_l_upperarm_twist_02_l {parent=dyn_arm_l_upperarm_l, loc=<13.7941 0 0>, rot=<0 0 0 1>}
 dyn_upperarm_l_AIM_MSTR {parent=dyn_arm_l_clavicle_l, loc=<15.5111 0 0>, rot=<0.0155 -0.3235 -0.0064 -0.9461>}
 dyn_upperarm_l_AIM {parent=dyn_upperarm_l_AIM_MSTR, loc=<0 0 0>, rot=<0 0 0 1>}
 dyn_spike_b {parent=dyn_upperarm_l_AIM, loc=<-4.513 5.4116 4.4679>, rot=<-0.1343 -0.7826 0.398 0.4594>}
 dyn_spike_b_01 {parent=dyn_spike_b, loc=<8.9336 0 0>, rot=<0 0 0 1>}
 dyn_spike_a {parent=dyn_upperarm_l_AIM, loc=<-4.117 -5.7409 4.465>, rot=<0.162 -0.7772 -0.3812 0.4738>}
 dyn_spike_a_01 {parent=dyn_spike_a, loc=<8.9381 0 0>, rot=<0 0 0 1>}
]
variables [
 Arm Twists {type=double, default=0}
 Dual Wield {type=double, default=0}
 Emote And Pickaxe {type=double, default=0}
 Parachute Arm {type=double, default=0}
 Rifle {type=double, default=0}
 Rifle Arm Up {type=double, default=0}
 Weapon Swap {type=double, default=0}
 skydive float {type=double, default=0}
]
